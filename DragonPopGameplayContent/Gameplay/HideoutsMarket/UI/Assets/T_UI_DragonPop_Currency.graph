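# Translucent UI material for the DragonPop currency icon in the Hideouts Market, built from the Battle Royale gold bars icon texture despite its T_ prefix.
# It samples the icon twice, adding a vertically offset, dimmed copy of the alpha to the original alpha, which suggests a drop shadow beneath the icon.
# A lerp from black to the icon's RGB, using that same RGB as alpha, squares the color for a darker, higher-contrast look.
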Material /DragonPopGameplayContent/Gameplay/HideoutsMarket/UI/Assets/T_UI_DragonPop_Currency
graph {
 properties {bAutomaticallySetUsageInEditor=false, blendMode=BLEND_Translucent, materialDomain=MD_UI, usedShadingModels=Default Lit}
 nodes [
  n1 = TextureSample() {texture=/Game/UI/Foundation/Textures/Icons/Athena/T-T-Icon-BR-GoldBars-UI-Icon}
  n2 = Add(A=TextureCoordinate(), B=Constant2Vector() {g=-0.092})
  Add(
   A=@n1.A
   B=Multiply(A=TextureSample(UVs=@n2) {texture=&T-T-Icon-BR-GoldBars-UI-Icon}.A) {constB=0.7}
  )
  LinearInterpolate(A=Constant3Vector(), B=@n1.RGB, Alpha=@n1.RGB)
 ]
}
